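# Physics asset for the TeddyPermit_Able pickaxe cosmetic, defining capsule bodies and joints for its dangling chain and tassel.
# A four-link chain hangs from base_chain with wide swing and tighter twist limits.
# A three-segment tassel hangs from base_tassel with varied swing and very tight twist, its capsules widening toward the tip.
# Larger right thigh and calf capsules have no constraints, which suggests they serve as leg colliders.
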PhysicsAsset /BRCosmetics/Weapons/FORT_Melee/Pickaxe_TeddyPermit_Able/Meshes/TeddyPermit_Able_Axe_Physics_Chain
bodies {
 base_chain [{shapeType=Capsule, length=41.5, radius=3, center=<-2.5278 0 0>}]
 base_tassel [{shapeType=Capsule, length=1, radius=1}]
 calf_r [{shapeType=Capsule, length=36, radius=8.005, shapeName=calf_r_capsule, center=<26.0485 0 0>, rotation=<rot -90 179.9996 0>}]
 dyn_chain_01 [{shapeType=Capsule, length=1, radius=2.5, center=<3.2879 0 0>, rotation=<rot 90 -0.0002 0>}]
 dyn_chain_02 [{shapeType=Capsule, length=1, radius=2.5, center=<3.2879 0 0>, rotation=<rot 90 -0.0002 0>}]
 dyn_chain_03 [{shapeType=Capsule, length=1, radius=2.5, center=<3.2879 0 0>, rotation=<rot 90 -0.0002 0>}]
 dyn_chain_04 [{shapeType=Capsule, length=1, radius=2.5, center=<3.2879 0 0>, rotation=<rot 90 -0.0002 0>}]
 dyn_tassel_01 [{shapeType=Capsule, length=1, radius=1, center=<1.9291 0 0>, rotation=<rot 90 0.0002 0>}]
 dyn_tassel_02 [{shapeType=Capsule, length=2, radius=2, center=<3.4049 0 0>, rotation=<rot 90 0.0002 0>}]
 dyn_tassel_03 [{shapeType=Capsule, length=2, radius=3, center=<4.3903 0 0>, rotation=<rot 90 0.0002 0>}]
 thigh_r [{shapeType=Capsule, length=27.5, radius=8, center=<20.7469 -0.0001 0>, rotation=<rot -90 -0.0002 0>}]
}
constraints [
 {bone1Name=dyn_chain_01, bone2Name=base_chain, swing1LimitDegrees=75, swing2LimitDegrees=75, twistLimitDegrees=15}
 {bone1Name=dyn_chain_02, bone2Name=dyn_chain_01, swing1LimitDegrees=75, swing2LimitDegrees=75, twistLimitDegrees=15}
 {bone1Name=dyn_chain_03, bone2Name=dyn_chain_02, swing1LimitDegrees=75, swing2LimitDegrees=75, twistLimitDegrees=15}
 {bone1Name=dyn_chain_04, bone2Name=dyn_chain_03, swing1LimitDegrees=75, swing2LimitDegrees=75, twistLimitDegrees=15}
 {bone1Name=dyn_tassel_01, bone2Name=base_tassel, swing1LimitDegrees=60, swing2LimitDegrees=60, twistLimitDegrees=5}
 {bone1Name=dyn_tassel_02, bone2Name=dyn_tassel_01, swing1LimitDegrees=25, swing2LimitDegrees=25, twistLimitDegrees=5}
 {bone1Name=dyn_tassel_03, bone2Name=dyn_tassel_02, swing1LimitDegrees=45, swing2LimitDegrees=45, twistLimitDegrees=5}
]
