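# Animation Blueprint for the TeddyPermit Able pickaxe, adding secondary motion to its bow, chain and tassel.
# A shared start pose offsets the handle, sets right-leg bones in world space and simulates the bow with AnimDynamics.
# When RigidsActive is set, chain and tassel use AnimDynamics in the front end or RigidBody otherwise; when unset, trail controllers and bone offsets take over.
# The event graph samples the parent mesh's leg transforms and derives RigidsActive from rigid-body enablement or LOD.
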AnimBlueprint /BRCosmetics/Weapons/FORT_Melee/Pickaxe_TeddyPermit_Able/Meshes/TeddyPermit_Able_Axe_AnimBP
target_skeleton=&/TeddyPermit_Able_Axe_Skeleton
graphs [
 AnimGraph {nodes [
   n1 = Use cached pose 'RbanSetup'()
   n2 = Use cached pose 'FESetup'()
   n3 = Get bIsInFrontEnd()
   n4 = Get bIsInFrontEnd()
   n5 = Trail controller - Bone: dyn_chain_02!(ComponentPose=Local To Component(LocalPose=Use cached pose 'StartPose'()), bAlphaBoolEnabled:bool=@n4) -> (Pose) {
    alphaInputType=Bool
    relaxationSpeedScale=2.5
    rotationLimits [{limitMax=<180 180 180>}, {limitMax=<180 180 180>}]
    rotationOffsets [<0 0 0>, <0 0 0>]
    trailBone {boneName=dyn_chain_02}
    trailRelaxationSpeed {editorCurveData {keys [
       {interpMode=RCIM_Cubic, tangentMode=RCTM_Break, time=0, value=10}
       {arriveTangent=-0.4334, interpMode=RCIM_Cubic, tangentMode=RCTM_Break, time=1, value=5}
      ]}}
    bLimitStretch=~
   }
   n6 = Get NegPhysicsAlpha() -> (NegPhysicsAlpha:double)
   n7 = Trail controller - Bone: dyn_chain_02!(ComponentPose=@n5, Alpha:float=@n6) -> (Pose) {
    maxDeltaTime=0.0333
    relaxationSpeedScale=2.5
    rotationLimits [{limitMax=<180 180 180>}, {limitMax=<180 180 180>}]
    rotationOffsets [<0 0 0>, <0 0 0>]
    trailBone {boneName=dyn_chain_02}
    trailRelaxationSpeed {editorCurveData {keys [{interpMode=RCIM_Cubic, tangentMode=RCTM_User, time=0, value=10}, {interpMode=RCIM_Cubic, tangentMode=RCTM_User, time=1, value=5}]}}
   }
   n8 = Trail controller - Bone: dyn_tassel_03!(ComponentPose=@n7, bAlphaBoolEnabled:bool=@n4) -> (Pose) {
    alphaInputType=Bool
    relaxationSpeedScale=2.5
    rotationLimits [{limitMax=<180 180 180>}, {limitMax=<180 180 180>}]
    rotationOffsets [<0 0 0>, <0 0 0>]
    trailBone {boneName=dyn_tassel_03}
    trailRelaxationSpeed {editorCurveData {keys [{interpMode=RCIM_Cubic, tangentMode=RCTM_User, time=0, value=10}, {interpMode=RCIM_Cubic, tangentMode=RCTM_User, time=1, value=5}]}}
    bLimitStretch=~
   }
   n9 = Trail controller - Bone: dyn_tassel_03!(ComponentPose=@n8, Alpha:float=@n6) -> (Pose) {
    maxDeltaTime=0.0333
    relaxationSpeedScale=2.5
    rotationLimits [{limitMax=<180 180 180>}, {limitMax=<180 180 180>}]
    rotationOffsets [<0 0 0>, <0 0 0>]
    trailBone {boneName=dyn_tassel_03}
    trailRelaxationSpeed {editorCurveData {keys [{interpMode=RCIM_Cubic, tangentMode=RCTM_User, time=0, value=10}, {interpMode=RCIM_Cubic, tangentMode=RCTM_User, time=1, value=5}]}}
   }
   n10 = Get bIsOnPawn() -> (bIsOnPawn:bool)
   n11 = "Transform (Modify) Bone - Bone: base_chain"!(ComponentPose=@n9, bAlphaBoolEnabled:bool=@n10, Translation="0,0,0", Rotation="-20,10,0", Scale="1,1,1") -> (Pose) {
    alphaInputType=Bool
    rotationMode=BMM_Additive
    rotationSpace=BCS_BoneSpace
    translationMode=BMM_Additive
    translationSpace=BCS_BoneSpace
    boneToModify {boneName=base_chain}
    rotation {pitch=-20, yaw=10}
   }
   n12 = "Transform (Modify) Bone - Bone: base_tassel"!(ComponentPose=@n11, bAlphaBoolEnabled:bool=@n10, Translation="0,0,0", Rotation="-110,-20,0", Scale="1,1,1") -> (Pose) {
    alphaInputType=Bool
    rotationMode=BMM_Additive
    rotationSpace=BCS_BoneSpace
    translationMode=BMM_Additive
    translationSpace=BCS_BoneSpace
    boneToModify {boneName=base_tassel}
    rotation {pitch=-110, yaw=-20}
   }
   n13 = NOT Boolean(A=@n10)
   n14 = "Transform (Modify) Bone - Bone: base_chain"(ComponentPose=@n12, bAlphaBoolEnabled=@n13, Translation="0,0,0", Rotation="60,90,45", Scale="1,1,1") {
    alphaInputType=Bool
    rotationMode=BMM_Additive
    rotationSpace=BCS_BoneSpace
    translationMode=BMM_Additive
    translationSpace=BCS_BoneSpace
    boneToModify {boneName=base_chain}
    rotation=<rot 60 90 45>
   }
   n15 = "Transform (Modify) Bone - Bone: base_tassel"(ComponentPose=@n14, bAlphaBoolEnabled=@n13, Translation="0,0,0", Rotation="0,30,0", Scale="1,1,1") {
    alphaInputType=Bool
    rotationMode=BMM_Additive
    rotationSpace=BCS_BoneSpace
    translationMode=BMM_Additive
    translationSpace=BCS_BoneSpace
    boneToModify {boneName=base_tassel}
    rotation {yaw=30}
   }
   Output Pose(Result=Blend Poses by bool(
     BlendPose_0=Blend Poses by bool(BlendPose_0=@n2, BlendPose_1=@n1, bActiveValue=@n3)
     BlendPose_1=Component To Local(ComponentPose=@n15)
     bActiveValue=Get RigidsActive() -> (RigidsActive:bool)
    ) {blendType=Linear}) {layerGroup=None}
   n16 = "Transform (Modify) Bone - Bone: Melee_TwoHanded_Handle"!(ComponentPose=Local To Component(LocalPose=Input Pose()), Alpha=1, Translation="0,-2,0") -> (Pose) {translationMode=BMM_Additive, translationSpace=BCS_BoneSpace, boneToModify {boneName=Melee_TwoHanded_Handle}, translation {y=-2}}
   n17 = "Transform (Modify) Bone - Bone: thigh_r"(ComponentPose=@n16, Alpha=1, Translation="0,0,0", Rotation="0,0,0", Scale="1,1,1") {
    rotationMode=BMM_Replace
    rotationSpace=BCS_WorldSpace
    translationMode=BMM_Replace
    translationSpace=BCS_WorldSpace
    boneToModify {boneName=thigh_r}
   }
   n18 = "Transform (Modify) Bone - Bone: calf_r"(ComponentPose=@n17, Alpha=1, Translation="0,0,0", Rotation="0,0,0", Scale="1,1,1") {
    rotationMode=BMM_Replace
    rotationSpace=BCS_WorldSpace
    translationMode=BMM_Replace
    translationSpace=BCS_WorldSpace
    boneToModify {boneName=calf_r}
   }
   n19 = Anim Dynamics - Bone: dyn_bow_01(ComponentPose=@n18, Alpha=1, ExternalForce="0,0,0") {
    angularDampingOverride=1
    angularSpringConstant=12000
    bAngularSpring=true
    bOverrideAngularDamping=true
    bUseGravityOverride=true
    gravityScale=0.35
    simulationSpace=BoneRelative
    boundBone {boneName=dyn_bow_01}
    componentAppliedLinearAccClamp=<1800 1800 1800>
    componentLinearAccScale=<1 0.3 1>
    componentLinearVelScale=<1 0.5 1>
    gravityOverride {z=-980}
    physicsBodyDefinitions [{
      boundBone {boneName=dyn_bow_01}
      boxExtents=<15 25 25>
      constraintSetup {angularLimitsMax=<5 5 45>, angularLimitsMin=<-5 -5 -90>, angularTarget=<1 0 0>}
      localJointOffset=<8 0 0>
     }]
    relativeSpaceBone {boneName=b_AR_Root}
    simSpaceSettings {maxAngularAcceleration=32, maxAngularVelocity=8, simSpaceAngularAlpha=1}
   }
   Save cached pose 'StartPose'(Pose=Component To Local(ComponentPose=@n19)) {cacheName=StartPose}
   n20 = RigidBody(ComponentPose=Local To Component(LocalPose=Use cached pose 'StartPose'())) {
    bOverrideWorldGravity=true
    overridePhysicsAsset=&/TeddyPermit_Able_Axe_Physics_Chain
    simulationSpace=BaseBoneSpace
    baseBoneRef {boneName=Melee_TwoHanded_Handle}
    overrideWorldGravity {z=-980}
    simSpaceSettings {
     dampingAlpha=0
     maxAngularAcceleration=28
     maxAngularVelocity=7
     maxLinearAcceleration=900
     maxLinearVelocity=600
     worldAlpha=1
     externalLinearDragV=<1 1 1>
    }
   }
   Save cached pose 'RbanSetup'(Pose=Component To Local(ComponentPose=@n20)) {cacheName=RbanSetup}
   n21 = Anim Dynamics - Bone: dyn_chain_01(ComponentPose=Local To Component(LocalPose=Use cached pose 'StartPose'()), Alpha=1, ExternalForce="0,0,0") {
    angularDampingOverride=1
    angularSpringConstant=40000
    bOverrideAngularDamping=true
    bUseGravityOverride=true
    gravityScale=0.35
    simulationSpace=BoneRelative
    boundBone {boneName=dyn_chain_01}
    componentAppliedLinearAccClamp=<1800 1800 1800>
    componentLinearAccScale=<0.3 0.3 0.3>
    componentLinearVelScale=<0.3 0.3 0.3>
    gravityOverride {z=-1200}
    physicsBodyDefinitions [{
      boundBone {boneName=dyn_chain_01}
      boxExtents=<12.5 15 15>
      constraintSetup {angularLimitsMax=<2 75 75>, angularLimitsMin=<-2 -75 -75>, angularTarget=<1 0 0>}
      localJointOffset=<3 0 0>
     }]
    relativeSpaceBone {boneName=b_AR_Root}
    simSpaceSettings {maxAngularAcceleration=48, maxAngularVelocity=12, simSpaceAngularAlpha=0.75}
   }
   n22 = Anim Dynamics - Bone: dyn_chain_02(ComponentPose=@n21, Alpha=1, ExternalForce="0,0,0") {
    angularDampingOverride=1
    angularSpringConstant=40000
    bOverrideAngularDamping=true
    bUseGravityOverride=true
    gravityScale=0.35
    simulationSpace=BoneRelative
    boundBone {boneName=dyn_chain_02}
    componentAppliedLinearAccClamp=<1800 1800 1800>
    componentLinearAccScale=<0.3 0.3 0.3>
    componentLinearVelScale=<0.3 0.3 0.3>
    gravityOverride {z=-1200}
    physicsBodyDefinitions [{
      boundBone {boneName=dyn_chain_02}
      boxExtents=<12.5 15 15>
      constraintSetup {angularLimitsMax=<2 75 75>, angularLimitsMin=<-2 -75 -75>, angularTarget=<1 0 0>}
      localJointOffset=<5 0 0>
     }]
    relativeSpaceBone {boneName=b_AR_Root}
    simSpaceSettings {maxAngularAcceleration=48, maxAngularVelocity=12, simSpaceAngularAlpha=0.75}
   }
   n23 = Anim Dynamics - Bone: dyn_chain_03(ComponentPose=@n22, Alpha=1, ExternalForce="0,0,0") {
    angularDampingOverride=1
    angularSpringConstant=40000
    bOverrideAngularDamping=true
    bUseGravityOverride=true
    gravityScale=0.35
    simulationSpace=BoneRelative
    boundBone {boneName=dyn_chain_03}
    componentAppliedLinearAccClamp=<1800 1800 1800>
    componentLinearAccScale=<0.3 0.3 0.3>
    componentLinearVelScale=<0.3 0.3 0.3>
    gravityOverride {z=-1200}
    physicsBodyDefinitions [{
      boundBone {boneName=dyn_chain_03}
      boxExtents=<12.5 15 15>
      constraintSetup {angularLimitsMax=<2 75 75>, angularLimitsMin=<-2 -75 -75>, angularTarget=<1 0 0>}
      localJointOffset=<5 0 0>
     }]
    relativeSpaceBone {boneName=b_AR_Root}
    simSpaceSettings {maxAngularAcceleration=48, maxAngularVelocity=12, simSpaceAngularAlpha=0.75}
   }
   n24 = Anim Dynamics - Bone: dyn_chain_04(ComponentPose=@n23, Alpha=1, ExternalForce="0,0,0") {
    angularDampingOverride=1
    angularSpringConstant=40000
    bOverrideAngularDamping=true
    bUseGravityOverride=true
    gravityScale=0.35
    simulationSpace=BoneRelative
    boundBone {boneName=dyn_chain_04}
    componentAppliedLinearAccClamp=<1800 1800 1800>
    componentLinearAccScale=<0.3 0.3 0.3>
    componentLinearVelScale=<0.3 0.3 0.3>
    gravityOverride {z=-1200}
    physicsBodyDefinitions [{
      boundBone {boneName=dyn_chain_04}
      boxExtents=<12.5 15 15>
      constraintSetup {angularLimitsMax=<2 75 75>, angularLimitsMin=<-2 -75 -75>, angularTarget=<1 0 0>}
      localJointOffset=<5 0 0>
     }]
    relativeSpaceBone {boneName=b_AR_Root}
    simSpaceSettings {maxAngularAcceleration=48, maxAngularVelocity=12, simSpaceAngularAlpha=0.75}
   }
   n25 = Anim Dynamics - Bone: dyn_tassel_01(ComponentPose=@n24, Alpha=1, ExternalForce="0,0,0") {
    angularDampingOverride=1
    angularSpringConstant=40000
    bOverrideAngularDamping=true
    bUseGravityOverride=true
    gravityScale=0.35
    simulationSpace=BoneRelative
    boundBone {boneName=dyn_tassel_01}
    componentAppliedLinearAccClamp=<1800 1800 1800>
    componentLinearAccScale=<0.1 0.1 0.1>
    componentLinearVelScale=<0.1 0.1 0.1>
    gravityOverride {z=-980}
    physicsBodyDefinitions [{
      boundBone {boneName=dyn_tassel_01}
      boxExtents=<10 15 15>
      constraintSetup {angularLimitsMax=<2.5 90 45>, angularLimitsMin=<-2.5 -10 -45>, angularTarget=<1 0 0>}
      localJointOffset=<3 0 0>
     }]
    relativeSpaceBone {boneName=b_AR_Root}
    simSpaceSettings {maxAngularAcceleration=35, maxAngularVelocity=8, simSpaceAngularAlpha=0.75}
   }
   n26 = Anim Dynamics - Bone: dyn_tassel_02(ComponentPose=@n25, Alpha=1, ExternalForce="0,0,0") {
    angularDampingOverride=1
    angularSpringConstant=40000
    bOverrideAngularDamping=true
    bUseGravityOverride=true
    gravityScale=0.35
    simulationSpace=BoneRelative
    boundBone {boneName=dyn_tassel_02}
    componentAppliedLinearAccClamp=<1800 1800 1800>
    componentLinearAccScale=<0.1 0.1 0.1>
    componentLinearVelScale=<0.1 0.1 0.1>
    gravityOverride {z=-980}
    physicsBodyDefinitions [{
      boundBone {boneName=dyn_tassel_02}
      boxExtents=<10 15 15>
      constraintSetup {angularLimitsMax=<2.5 25 25>, angularLimitsMin=<-2.5 -25 -25>, angularTarget=<1 0 0>}
      localJointOffset=<5 0 0>
     }]
    relativeSpaceBone {boneName=b_AR_Root}
    simSpaceSettings {maxAngularAcceleration=35, maxAngularVelocity=8, simSpaceAngularAlpha=0.75}
   }
   n27 = Anim Dynamics - Bone: dyn_tassel_03(ComponentPose=@n26, Alpha=1, ExternalForce="0,0,0") {
    angularDampingOverride=1
    angularSpringConstant=40000
    bOverrideAngularDamping=true
    bUseGravityOverride=true
    gravityScale=0.35
    simulationSpace=BoneRelative
    boundBone {boneName=dyn_tassel_03}
    componentAppliedLinearAccClamp=<1800 1800 1800>
    componentLinearAccScale=<0.1 0.1 0.1>
    componentLinearVelScale=<0.1 0.1 0.1>
    gravityOverride {z=-980}
    physicsBodyDefinitions [{
      boundBone {boneName=dyn_tassel_03}
      boxExtents=<10 15 15>
      constraintSetup {angularLimitsMax=<2.5 25 25>, angularLimitsMin=<-2.5 -25 -25>, angularTarget=<1 0 0>}
      localJointOffset=<7 0 0>
     }]
    relativeSpaceBone {boneName=b_AR_Root}
    simSpaceSettings {maxAngularAcceleration=35, maxAngularVelocity=8, simSpaceAngularAlpha=0.75}
   }
   Save cached pose 'FESetup'(Pose=Component To Local(ComponentPose=@n27)) {cacheName=FESetup}
  ]}
 EventGraph {nodes [
   Event BlueprintUpdateAnimation()
   > n1 = Sequence!(execute) -> (then_0, then_1, then_2:Exec, then_3:Exec, then_4:Exec)
   n2 = GetPartSkeletalMeshComponent()
   > Is Valid(InputObject=@n2)
   n3 = GetAttachParent!(self:Scene Component Object Reference=@n2) -> (ReturnValue:Scene Component Object Reference)
   > Is Valid(InputObject=@n3)
   > Set Body_Mesh(Body_Mesh:Scene Component Object Reference=@n3) -> (Output_Get:Scene Component Object Reference)
   n4 = Get Body_Mesh() -> (Body_Mesh:Scene Component Object Reference)
   n5 = BreakTransform(InTransform=GetSocketTransform(self=@n4, InSocketName=thigh_r, TransformSpace=RTS_World))
   Set Thigh_R_Location(execute=@n1.then_1, Thigh_R_Location:Vector=@n5) -> (Output_Get:Vector)
   > Set Thigh_R_Rotation(Thigh_R_Rotation:Rotator=@n5.Rotation) -> (Output_Get:Rotator)
   n6 = BreakTransform(InTransform=GetSocketTransform(self=@n4, InSocketName=calf_r, TransformSpace=RTS_World))
   Set Calf_R_Location(execute=@n1.then_2, Calf_R_Location:Vector=@n6) -> (Output_Get:Vector)
   > Set Calf_R_Rotation(Calf_R_Rotation:Rotator=@n6.Rotation) -> (Output_Get:Rotator)
   n7 = Branch(execute=@n1.then_4, Condition=Get bIsInFrontEnd())
   > Set RigidsFE_Alpha(RigidsFE_Alpha:double=MapRangeClamped(Value=GetCurveValue(CurveName=PhysicsOff), InRangeA=0, InRangeB=1, OutRangeA=1, OutRangeB=0)) -> (Output_Get:double)
   > Set NegPhysicsAlpha(NegPhysicsAlpha:double=0) -> (Output_Get:double)
   Set RigidsFE_Alpha(execute=@n7.else, RigidsFE_Alpha=0)
   n8 = Get CurrentLOD()
   n9 = Branch(execute=@n1.then_3, Condition=OR Boolean(A=Get bRigidBodyEnabled(), B="integer > integer"(A=@n8, B=1) {is=title K2Node_CallFunction}))
   > Set RigidsActive(RigidsActive:bool=true) -> (Output_Get:bool)
   Set RigidsActive(execute=@n9.else, RigidsActive=false)
  ]}
]
variables [
 Body_Mesh {type=SceneComponent, default=null}
 Calf_R_Location {type=Vector}
 Calf_R_Rotation {type=Rotator}
 NegPhysicsAlpha {type=number, default=1}
 RigidsActive {type=boolean}
 RigidsFE_Alpha {type=number}
 Thigh_R_Location {type=Vector}
 Thigh_R_Rotation {type=Rotator}
]
